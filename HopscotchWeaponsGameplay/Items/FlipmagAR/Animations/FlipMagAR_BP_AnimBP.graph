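# Animation Blueprint for the FlipMag assault rifle weapon mesh that holds its pose according to whether the next reload is fast or slow.
# ReloadFast and ReloadSlow anim notifies set the Is Next Reload Fast flag, which starts true.
# An instant blend by bool picks the ref pose when true or non-looping pose B when false, feeding DefaultSlot for montages.
# Initialization caches the owning weapon actor, which nothing else uses.
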AnimBlueprint /HopscotchWeaponsGameplay/Items/FlipmagAR/Animations/FlipMagAR_BP_AnimBP
target_skeleton=/HopscotchWeaponsGameplay/Items/FlipmagAR/Meshes/Assault_FlipMag_Rifle_Skeleton
graphs [
 AnimGraph {nodes [
   n1 = Local Space Ref Pose()
   n2 = Sequence Player 'assault_flipmag_poseB_WEP'() {bLoopAnimation=false, sequence=&/assault_flipmag_poseB_WEP}
   n3 = Get Is Next Reload Fast() -> (Is Next Reload Fast:bool)
   Output Pose(Result=Slot 'DefaultSlot'(Source=Blend Poses by bool(BlendPose_0=@n1, BlendPose_1=@n2, BlendTime_0=0, BlendTime_1=0, bActiveValue=@n3) {blendType=Linear})) {layerGroup=None}
  ]}
 EventGraph {nodes [
   Event BlueprintInitializeAnimation()
   n1 = GetOwningActor()
   > Is Valid(InputObject=@n1)
   > n2 = Cast To B_Assault_FlipMag_Athena!(execute, Object=@n1) -> (then, CastFailed, AsB Assault Flip Mag Athena:B Assault Flip Mag Athena Object Reference) {is=title K2Node_DynamicCast}
   > Set As B Assault Flip Mag Athena(As B Assault Flip Mag Athena:B Assault Flip Mag Athena Object Reference=@n2) -> (Output_Get:B Assault Flip Mag Athena Object Reference)
   AnimNotify_ReloadFast() {is=title K2Node_Event}
   > Set Is Next Reload Fast(Is Next Reload Fast:bool=true) -> (Output_Get:bool)
   AnimNotify_ReloadSlow() {is=title K2Node_Event}
   > Set Is Next Reload Fast(Is Next Reload Fast=false)
   Get As B Assault Flip Mag Athena() -> (As B Assault Flip Mag Athena:B Assault Flip Mag Athena Object Reference)
  ]}
]
variables [
 As B Assault Flip Mag Athena {type=B_Assault_FlipMag_Athena_C, default=null}
 Is Next Reload Fast {type=boolean, default=true}
]
