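# First-person fire sound cue for the FlipMag AR weapon, in the ranged-weapon fire sound class.
# A quality switch sends its first two slots to five random close shots layered with a quieter, fixed-delay random shell-casing impact.
# The last two slots use a lighter branch of three random shots without casings.
# First-person attenuation and owner-limited concurrency that stops the oldest voice apply, and the debug flag is on.
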SoundCue /HopscotchWeaponsGameplay/Items/FlipmagAR/Sounds/FlipMag_Fire_1P_Cue
root=QualityLevel_0
nodes [
 {id=Delay_0, children [Random_1], properties {delayMax=0.6, delayMin=0.6}}
 {id=Mixer_0, children [Mixer_1,Modulator_1], properties {inputVolume [1,0.6]}}
 {id=Mixer_1, children [Modulator_0], properties {inputVolume [1]}}
 {id=Mixer_2, children [Modulator_2], properties {inputVolume [1]}}
 {id=Modulator_0, children [Random_0], properties {volumeMax=1.7, volumeMin=1.6}}
 {id=Modulator_1, children [Delay_0]}
 {id=Modulator_2, children [Random_4], properties {volumeMax=1.7, volumeMin=1.6}}
 {id=QualityLevel_0, children [Mixer_0,Mixer_0,Mixer_2,Mixer_2]}
 {id=Random_0, children [WavePlayer_1,WavePlayer_2,WavePlayer_3,WavePlayer_4,WavePlayer_5], properties {weights [1,1,1,1,1]}}
 {
  id=Random_1
  children [WavePlayer_0, WavePlayer_6, WavePlayer_7, WavePlayer_8, WavePlayer_9, WavePlayer_10, WavePlayer_11]
  properties {weights [1,1,1,1,1,1,1]}
 }
 {id=Random_4, children [WavePlayer_31, WavePlayer_32, WavePlayer_33], properties {weights [1,1,1]}}
 {id=WavePlayer_0, properties {soundWaveAssetPtr=&/FlipMag_Reload_Casing_Impact_01}}
 {id=WavePlayer_1, properties {soundWaveAssetPtr=&/FlipMag_Fire_Close_Full_01}}
 {id=WavePlayer_10, properties {soundWaveAssetPtr=&/FlipMag_Reload_Casing_Impact_06}}
 {id=WavePlayer_11, properties {soundWaveAssetPtr=&/FlipMag_Reload_Casing_Impact_07}}
 {id=WavePlayer_2, properties {soundWaveAssetPtr=&/FlipMag_Fire_Close_Full_02}}
 {id=WavePlayer_3, properties {soundWaveAssetPtr=&/FlipMag_Fire_Close_Full_03}}
 {id=WavePlayer_31, properties {soundWaveAssetPtr=&FlipMag_Fire_Close_Full_01}}
 {id=WavePlayer_32, properties {soundWaveAssetPtr=&FlipMag_Fire_Close_Full_02}}
 {id=WavePlayer_33, properties {soundWaveAssetPtr=&FlipMag_Fire_Close_Full_03}}
 {id=WavePlayer_4, properties {soundWaveAssetPtr=&/FlipMag_Fire_Close_Full_04}}
 {id=WavePlayer_5, properties {soundWaveAssetPtr=&/FlipMag_Fire_Close_Full_05}}
 {id=WavePlayer_6, properties {soundWaveAssetPtr=&/FlipMag_Reload_Casing_Impact_02}}
 {id=WavePlayer_7, properties {soundWaveAssetPtr=&/FlipMag_Reload_Casing_Impact_03}}
 {id=WavePlayer_8, properties {soundWaveAssetPtr=&/FlipMag_Reload_Casing_Impact_04}}
 {id=WavePlayer_9, properties {soundWaveAssetPtr=&/FlipMag_Reload_Casing_Impact_05}}
]
properties {
 attenuationSettings=&Fort_Weapon_Projectile_FirstPerson_Athena_Att
 bDebug=true
 maxDistance=2097152
 priority=5
 soundClassObject=&Fort_WeaponRanged_Fire_1P_PSM
 volumeMultiplier=0.62
 attenuationOverrides {manualReverbSendLevel=-1.0165, reverbSendMethod=Manual}
 concurrencyOverrides {bLimitToOwner=true, maxCount=2, resolutionRule=StopOldest}
 concurrencySet [&Weapon_Gunfire_FirstPerson_Conc]
}
