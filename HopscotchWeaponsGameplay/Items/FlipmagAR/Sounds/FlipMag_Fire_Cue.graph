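# Gunfire sound cue for the Flipmag AR, layering close, medium and distant shot recordings with third-person and distant weapon fire sound classes.
# Each layer picks a random wave, modulates its volume, then applies distance-specific attenuation and a sound class before mixing.
# A quality switch sends its first two levels to a five-variation mix and the last two to a leaner two-to-three-variation mix.
# Global and local gunfire concurrency sets are assigned.
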SoundCue /HopscotchWeaponsGameplay/Items/FlipmagAR/Sounds/FlipMag_Fire_Cue
root=QualityLevel_0
nodes [
 {id=Attenuation_0, children [Modulator_3], properties {attenuationSettings=&Fort_Weapon_Projectile_Close_Athena_Att}}
 {
  id=Attenuation_1
  children [Modulator_4]
  properties {attenuationSettings=&Fort_Weapon_Projectile_Medium_Mono_Athena_Att}
 }
 {
  id=Attenuation_2
  children [Modulator_5]
  properties {attenuationSettings=&Fort_Weapon_Projectile_Distant_Mono_Athena_Att}
 }
 {id=Attenuation_6, children [Modulator_6], properties {attenuationSettings=&Fort_Weapon_Projectile_Close_Athena_Att}}
 {
  id=Attenuation_7
  children [Modulator_7]
  properties {attenuationSettings=&Fort_Weapon_Projectile_Medium_Mono_Athena_Att}
 }
 {
  id=Attenuation_8
  children [Modulator_8]
  properties {attenuationSettings=&Fort_Weapon_Projectile_Distant_Mono_Athena_Att}
 }
 {id=Mixer_0, children [SoundClass_6,SoundClass_7,SoundClass_8], properties {inputVolume [0.85,0.6,0.35]}}
 {id=Mixer_2, children [SoundClass_0,SoundClass_1,SoundClass_2], properties {inputVolume [0.85,0.6,0.3]}}
 {id=Modulator_3, children [Random_0], properties {volumeMax=1.2, volumeMin=1.1}}
 {id=Modulator_4, children [Random_2], properties {volumeMax=2.2, volumeMin=2.1}}
 {id=Modulator_5, children [Random_3], properties {volumeMax=2, volumeMin=2}}
 {id=Modulator_6, children [Random_6], properties {volumeMax=1.2, volumeMin=1.1}}
 {id=Modulator_7, children [Random_7], properties {volumeMax=2.2, volumeMin=2.1}}
 {id=Modulator_8, children [Random_8], properties {volumeMax=2, volumeMin=2}}
 {id=QualityLevel_0, children [Mixer_2,Mixer_2,Mixer_0,Mixer_0]}
 {id=Random_0, children [WavePlayer_0,WavePlayer_1,WavePlayer_2,WavePlayer_3,WavePlayer_4], properties {weights [1,1,1,1,1]}}
 {id=Random_2, children [WavePlayer_5, WavePlayer_23, WavePlayer_24, WavePlayer_25, WavePlayer_26], properties {weights [1,1,1,1,1]}}
 {id=Random_3, children [WavePlayer_27, WavePlayer_28, WavePlayer_29, WavePlayer_30, WavePlayer_31], properties {weights [1,1,1,1,1]}}
 {id=Random_6, children [WavePlayer_32, WavePlayer_33, WavePlayer_34], properties {weights [1,1,1]}}
 {id=Random_7, children [WavePlayer_37, WavePlayer_38], properties {weights [1,1]}}
 {id=Random_8, children [WavePlayer_42, WavePlayer_43], properties {weights [1,1]}}
 {
  id=SoundClass_0
  children [Attenuation_0]
  properties {soundClassOverride=&Fort_WeaponRanged_Fire_3P_PSM}
 }
 {
  id=SoundClass_1
  children [Attenuation_1]
  properties {soundClassOverride=&Fort_WeaponRanged_Fire_Distant}
 }
 {
  id=SoundClass_2
  children [Attenuation_2]
  properties {soundClassOverride=&Fort_WeaponRanged_Fire_Distant}
 }
 {
  id=SoundClass_6
  children [Attenuation_6]
  properties {soundClassOverride=&Fort_WeaponRanged_Fire_3P_PSM}
 }
 {
  id=SoundClass_7
  children [Attenuation_7]
  properties {soundClassOverride=&Fort_WeaponRanged_Fire_Distant}
 }
 {
  id=SoundClass_8
  children [Attenuation_8]
  properties {soundClassOverride=&Fort_WeaponRanged_Fire_Distant}
 }
 {id=WavePlayer_0, properties {soundWaveAssetPtr=&/FlipMag_Fire_Close_Full_01}}
 {id=WavePlayer_1, properties {soundWaveAssetPtr=&/FlipMag_Fire_Close_Full_02}}
 {id=WavePlayer_2, properties {soundWaveAssetPtr=&/FlipMag_Fire_Close_Full_03}}
 {id=WavePlayer_23, properties {soundWaveAssetPtr=&/FlipMag_Fire_Close_Medium_02}}
 {id=WavePlayer_24, properties {soundWaveAssetPtr=&/FlipMag_Fire_Close_Medium_03}}
 {id=WavePlayer_25, properties {soundWaveAssetPtr=&/FlipMag_Fire_Close_Medium_04}}
 {id=WavePlayer_26, properties {soundWaveAssetPtr=&/FlipMag_Fire_Close_Medium_05}}
 {id=WavePlayer_27, properties {soundWaveAssetPtr=&/FlipMag_Fire_Close_Distant_01}}
 {id=WavePlayer_28, properties {soundWaveAssetPtr=&/FlipMag_Fire_Close_Distant_02}}
 {id=WavePlayer_29, properties {soundWaveAssetPtr=&/FlipMag_Fire_Close_Distant_03}}
 {id=WavePlayer_3, properties {soundWaveAssetPtr=&/FlipMag_Fire_Close_Full_04}}
 {id=WavePlayer_30, properties {soundWaveAssetPtr=&/FlipMag_Fire_Close_Distant_04}}
 {id=WavePlayer_31, properties {soundWaveAssetPtr=&/FlipMag_Fire_Close_Distant_05}}
 {id=WavePlayer_32, properties {soundWaveAssetPtr=&FlipMag_Fire_Close_Full_01}}
 {id=WavePlayer_33, properties {soundWaveAssetPtr=&FlipMag_Fire_Close_Full_02}}
 {id=WavePlayer_34, properties {soundWaveAssetPtr=&FlipMag_Fire_Close_Full_03}}
 {id=WavePlayer_37, properties {soundWaveAssetPtr=&/FlipMag_Fire_Close_Medium_01}}
 {id=WavePlayer_38, properties {soundWaveAssetPtr=&FlipMag_Fire_Close_Medium_02}}
 {id=WavePlayer_4, properties {soundWaveAssetPtr=&/FlipMag_Fire_Close_Full_05}}
 {id=WavePlayer_42, properties {soundWaveAssetPtr=&FlipMag_Fire_Close_Distant_01}}
 {id=WavePlayer_43, properties {soundWaveAssetPtr=&FlipMag_Fire_Close_Distant_02}}
 {id=WavePlayer_5, properties {soundWaveAssetPtr=&FlipMag_Fire_Close_Medium_01}}
]
properties {
 bDebug=true
 maxDistance=27500
 pitchMultiplier=1.05
 volumeMultiplier=0.7
 concurrencyOverrides {bLimitToOwner=true, maxCount=2, resolutionRule=StopOldest}
 concurrencySet [&Weapon_Gunfire_Global_Conc, &Weapon_Gunfire_Local_Conc]
}
