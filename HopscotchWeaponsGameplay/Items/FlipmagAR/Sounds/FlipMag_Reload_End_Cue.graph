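# Sound cue for the end of the Flipmag AR weapon's reload, classed as weapon foley.
# A Random node picks among three reload-end wave variations with equal weight and feeds a Modulator left at its usual settings.
# The cue uses the shared weapon reload attenuation, lowers volume, slightly raises pitch and overrides concurrency.
# Its debug flag is switched on.
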SoundCue /HopscotchWeaponsGameplay/Items/FlipmagAR/Sounds/FlipMag_Reload_End_Cue
root=Modulator_0
nodes [
 {id=Modulator_0, children [Random_1]}
 {id=Random_1, children [WavePlayer_3,WavePlayer_4,WavePlayer_5], properties {weights [1,1,1]}}
 {id=WavePlayer_3, properties {soundWaveAssetPtr=&/FlipMag_Reload_End_01}}
 {id=WavePlayer_4, properties {soundWaveAssetPtr=&/FlipMag_Reload_End_02}}
 {id=WavePlayer_5, properties {soundWaveAssetPtr=&/FlipMag_Reload_End_03}}
]
properties {
 attenuationSettings=&Fort_Weapon_Reload_Att
 bDebug=true
 bOverrideConcurrency=true
 maxDistance=2250
 pitchMultiplier=1.1
 soundClassObject=&Fort_WeaponFoley
 volumeMultiplier=0.551
}
