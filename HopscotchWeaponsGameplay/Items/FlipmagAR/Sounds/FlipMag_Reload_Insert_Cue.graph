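# Sound cue for the magazine insert during a Flipmag AR reload, classed as weapon foley with reload attenuation.
# A mixer layers a random pick of three insert sounds, pitched up, with a random pick of two slam sounds, pitched down and delayed by a fixed interval.
# A top-level modulator sits over the mix, and the cue overrides concurrency and has its debug flag enabled.
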SoundCue /HopscotchWeaponsGameplay/Items/FlipmagAR/Sounds/FlipMag_Reload_Insert_Cue
root=Modulator_0
nodes [
 {id=Delay_0, children [Random_2], properties {delayMax=0.4, delayMin=0.4}}
 {id=Mixer_0, children [Modulator_1,Modulator_2], properties {inputVolume [1,1]}}
 {id=Modulator_0, children [Mixer_0]}
 {id=Modulator_1, children [Random_1], properties {pitchMax=1.1, pitchMin=1.1, volumeMax=1, volumeMin=1}}
 {id=Modulator_2, children [Delay_0], properties {pitchMax=0.9, pitchMin=0.9}}
 {id=Random_1, children [WavePlayer_3,WavePlayer_4,WavePlayer_5], properties {weights [1,1,1]}}
 {id=Random_2, children [WavePlayer_6,WavePlayer_7], properties {weights [1,1]}}
 {id=WavePlayer_3, properties {soundWaveAssetPtr=&/FlipMag_Reload_Insert_01}}
 {id=WavePlayer_4, properties {soundWaveAssetPtr=&/FlipMag_Reload_Insert_02}}
 {id=WavePlayer_5, properties {soundWaveAssetPtr=&/FlipMag_Reload_Insert_03}}
 {id=WavePlayer_6, properties {soundWaveAssetPtr=&/FlipMag_Reload_Insert_Slam_01}}
 {id=WavePlayer_7, properties {soundWaveAssetPtr=&/FlipMag_Reload_Insert_Slam_02}}
]
properties {
 attenuationSettings=&Fort_Weapon_Reload_Att
 bDebug=true
 bOverrideConcurrency=true
 maxDistance=2250
 soundClassObject=&Fort_WeaponFoley
 volumeMultiplier=0.601
}
