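# Sound cue for the fast magazine-insert reload foley of the FlipMag AR weapon.
# A Random node picks one of three equally weighted wave variants, which passes through a modulator that applies a fixed upward pitch shift.
# An outer modulator at default settings adds slight variation.
# The cue uses weapon reload attenuation and the weapon foley sound class, overrides concurrency, and has its debug flag enabled.
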SoundCue /HopscotchWeaponsGameplay/Items/FlipmagAR/Sounds/FlipMag_Reload_Insert_Fast_Cue
root=Modulator_0
nodes [
 {id=Modulator_0, children [Modulator_1]}
 {id=Modulator_1, children [Random_0], properties {pitchMax=1.1, pitchMin=1.1, volumeMax=1, volumeMin=1}}
 {id=Random_0, children [WavePlayer_0,WavePlayer_1,WavePlayer_2], properties {weights [1,1,1]}}
 {id=WavePlayer_0, properties {soundWaveAssetPtr=&/FlipMag_Reload_Insert_Fast_01}}
 {id=WavePlayer_1, properties {soundWaveAssetPtr=&/FlipMag_Reload_Insert_Fast_02}}
 {id=WavePlayer_2, properties {soundWaveAssetPtr=&/FlipMag_Reload_Insert_Fast_03}}
]
properties {
 attenuationSettings=&Fort_Weapon_Reload_Att
 bDebug=true
 bOverrideConcurrency=true
 maxDistance=2250
 soundClassObject=&Fort_WeaponFoley
 volumeMultiplier=0.802
}
