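# Sound cue for the low-ammo reload sound of the FlipMag AR weapon.
# A single wave player feeds the output directly with one low-ammo reload wave, with no randomization or modulation, at a boosted volume.
# It routes through the degraded ranged-weapon low-ammo sound class and overrides concurrency so new instances are blocked once the limit is reached.
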SoundCue /HopscotchWeaponsGameplay/Items/FlipmagAR/Sounds/FlipMag_Reload_LowAmmo_Cue
root=WavePlayer_0
nodes [{id=WavePlayer_0, properties {soundWaveAssetPtr=&/FlipMag_Reload_LowAmmo_02}}]
properties {
 bOverrideConcurrency=true
 maxDistance=2097152
 soundClassObject=/Game/Sounds/FortSoundClassesAndModes/Fort_WeaponRanged_LowAmmo_Degraded_PSM
 volumeMultiplier=1.4
 concurrencyOverrides {resolutionRule=PreventNew}
}
