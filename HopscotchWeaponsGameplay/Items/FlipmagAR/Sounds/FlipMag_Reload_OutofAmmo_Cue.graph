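# Sound cue for the FlipMag AR weapon, named for its out-of-ammo reload sound.
# A Random node picks between two equally weighted wave variations and feeds a Modulator that slightly varies volume while holding pitch fixed.
# The cue plays at reduced volume in the third-person ranged weapon fire sound class.
# Its concurrency override caps simultaneous instances at two, stopping the oldest first.
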SoundCue /HopscotchWeaponsGameplay/Items/FlipmagAR/Sounds/FlipMag_Reload_OutofAmmo_Cue
root=Modulator_0
nodes [
 {id=Modulator_0, children [Random_1], properties {pitchMax=1, pitchMin=1, volumeMax=0.95, volumeMin=0.85}}
 {id=Random_1, children [WavePlayer_4,WavePlayer_5], properties {weights [1,1]}}
 {id=WavePlayer_4, properties {soundWaveAssetPtr=&/FlipMag_Reload_OutofAmmo_01}}
 {id=WavePlayer_5, properties {soundWaveAssetPtr=&/FlipMag_Reload_OutofAmmo_02}}
]
properties {
 bOverrideConcurrency=true
 maxDistance=2097152
 soundClassObject=&Fort_WeaponRanged_Fire_3P_PSM
 volumeMultiplier=0.5
 attenuationOverrides {falloffDistance=1600}
 concurrencyOverrides {maxCount=2, resolutionRule=StopOldest}
}
