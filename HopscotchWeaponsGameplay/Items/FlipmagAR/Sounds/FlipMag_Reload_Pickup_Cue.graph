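# Sound cue for the FlipMag AR reload pickup sound, routed to the loot collect sound class with weapon pickup attenuation.
# A Random node picks one of three equally weighted wave variations and feeds it through a Modulator left at its default settings.
# A concurrency override allows a single instance at a time and blocks new plays while one is active, and overall volume is lowered.
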SoundCue /HopscotchWeaponsGameplay/Items/FlipmagAR/Sounds/FlipMag_Reload_Pickup_Cue
root=Modulator_0
nodes [
 {id=Modulator_0, children [Random_1]}
 {id=Random_1, children [WavePlayer_2,WavePlayer_4,WavePlayer_5], properties {weights [1,1,1]}}
 {id=WavePlayer_2, properties {soundWaveAssetPtr=&/FlipMag_Reload_Pickup_01}}
 {id=WavePlayer_4, properties {soundWaveAssetPtr=&/FlipMag_Reload_Pickup_02}}
 {id=WavePlayer_5, properties {soundWaveAssetPtr=&/FlipMag_Reload_Pickup_03}}
]
properties {
 attenuationSettings=&Fort_Weapon_Pickup_Att
 bOverrideConcurrency=true
 maxDistance=3500
 soundClassObject=&Fort_Loot_Collect
 volumeMultiplier=0.6
 concurrencyOverrides {maxCount=1, resolutionRule=PreventNew}
}
