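# Sound cue for the start of the FlipMag AR reload, routed to the weapon foley class with reload attenuation.
# A mixer layers one of three random reload-start waves with a quieter, pitched-up magazine impact, also picked from three variants and delayed half a second.
# An outer modulator wraps the whole mix, and the cue overrides concurrency and has its debug flag on.
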SoundCue /HopscotchWeaponsGameplay/Items/FlipmagAR/Sounds/FlipMag_Reload_Start_Cue
root=Modulator_0
nodes [
 {id=Delay_0, children [Random_0], properties {delayMax=0.5, delayMin=0.5}}
 {id=Mixer_0, children [Random_1,Modulator_1], properties {inputVolume [1,0.4]}}
 {id=Modulator_0, children [Mixer_0]}
 {id=Modulator_1, children [Delay_0], properties {pitchMax=1.15, pitchMin=1.05}}
 {id=Random_0, children [WavePlayer_0,WavePlayer_1,WavePlayer_2], properties {weights [1,1,1]}}
 {id=Random_1, children [WavePlayer_3,WavePlayer_4,WavePlayer_5], properties {weights [1,1,1]}}
 {id=WavePlayer_0, properties {soundWaveAssetPtr=&/FlipMag_Magazine_Impact_01}}
 {id=WavePlayer_1, properties {soundWaveAssetPtr=&/FlipMag_Magazine_Impact_02}}
 {id=WavePlayer_2, properties {soundWaveAssetPtr=&/FlipMag_Magazine_Impact_03}}
 {id=WavePlayer_3, properties {soundWaveAssetPtr=&/FlipMag_Reload_Start_01}}
 {id=WavePlayer_4, properties {soundWaveAssetPtr=&/FlipMag_Reload_Start_02}}
 {id=WavePlayer_5, properties {soundWaveAssetPtr=&/FlipMag_Reload_Start_03}}
]
properties {
 attenuationSettings=&Fort_Weapon_Reload_Att
 bDebug=true
 bOverrideConcurrency=true
 maxDistance=2250
 soundClassObject=&Fort_WeaponFoley
 volumeMultiplier=0.601
}
