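# First-person 5.1 MetaSound source for FlipMag AR gunfire.
# A fire-control subgraph emits gunshot triggers for main, low-ammo, mech, sub and optional (unpopulated) ADS sweetener layers, plus delayed casing impacts labelled OnStopFire but fired per shot.
# Layers mix per speaker group with master gains; aiming ducks main fire and boosts mech, sub and casings.
# OnFinished accumulates main, low-ammo and casing completion; concurrency stops the oldest.
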MetaSoundSource /HopscotchWeaponsGameplay/Items/FlipmagAR/Sounds/Settings/MSS_FlipMag_Weapon_Fire_1p
inputs [
 bUseADSSweeteners {type=Bool, access=Constructor}
 Character.Affiliation {type=Enum:FortAudioAffiliation}
 Character.Health {type=Float}
 Character.UniqueID {type=Int32}
 LowAmmo.FrontLRVolume {type=Float}
 LowAmmo.MaxPitch {type=Float, access=Constructor}
 LowAmmo.MaxVolume {type=Float, access=Constructor}
 LowAmmo.MinPitch {type=Float, access=Constructor}
 LowAmmo.MinVolume {type=Float, access=Constructor}
 LowAmmo.Sounds {type=WaveAsset:Array, access=Constructor}
 LowAmmo.Threshold {type=Int32}
 MainFire.FrontCenterVolume {type=Float}
 MainFire.FrontLRVolume {type=Float}
 MainFire.LFVolume {type=Float}
 MainFire.MaxPitch {type=Float, access=Constructor}
 MainFire.MaxVolume {type=Float, access=Constructor}
 MainFire.MinPitch {type=Float, access=Constructor}
 MainFire.MinVolume {type=Float, access=Constructor}
 MainFire.RearLRVolume {type=Float, access=Constructor}
 MainFire.Sounds {type=WaveAsset:Array, access=Constructor}
 Mech.bEnableMechLayer {type=Bool}
 Mech.FrontCenterVolume {type=Float}
 Mech.MaxPitch {type=Float, access=Constructor}
 Mech.MaxVolume {type=Float, access=Constructor}
 Mech.MinPitch {type=Float, access=Constructor}
 Mech.MinVolume {type=Float, access=Constructor}
 Mech.RandomTimeMax {type=Time, access=Constructor}
 Mech.RandomTimeMin {type=Time, access=Constructor}
 Mech.Sounds {type=WaveAsset:Array, access=Constructor}
 Mixing.ADSVolumeMultiply {type=Float}
 Mixing.Master.FrontCenterMasterVolume {type=Float}
 Mixing.Master.FrontLRMasterVolume {type=Float}
 Mixing.Master.LowFrequencyMasterVolume {type=Float}
 Mixing.Master.RearLRMasterVolume {type=Float}
 OnStop {type=Trigger}
 OnStopFire.FrontCenterVolume {type=Float}
 OnStopFire.FrontLRVolume {type=Float}
 OnStopFire.LFVolume {type=Float}
 OnStopFire.MaxPitch {type=Float, access=Constructor}
 OnStopFire.MaxVolume {type=Float, access=Constructor}
 OnStopFire.MinPitch {type=Float, access=Constructor}
 OnStopFire.MinVolume {type=Float, access=Constructor}
 OnStopFire.RandomTimeMax {type=Time}
 OnStopFire.RandomTimeMin {type=Time}
 OnStopFire.RearLRVolume {type=Float}
 OnStopFire.Sounds {type=WaveAsset:Array, access=Constructor}
 Sub.bEnable {type=Bool}
 Sub.LFVolume {type=Float}
 Sub.MaxPitch {type=Float, access=Constructor}
 Sub.MaxVolume {type=Float, access=Constructor}
 Sub.MinPitch {type=Float, access=Constructor}
 Sub.MinVolume {type=Float, access=Constructor}
 Sub.RandomTimeMax {type=Time}
 Sub.RandomTimeMin {type=Time}
 Sub.Sounds {type=WaveAsset:Array, access=Constructor}
 Sweetener.FrontCenterVolume {type=Float}
 Sweetener.FrontLRVolume {type=Float}
 Sweetener.MaxPitch {type=Float, access=Constructor}
 Sweetener.MaxVolume {type=Float, access=Constructor}
 Sweetener.MinPitch {type=Float, access=Constructor}
 Sweetener.MinVolume {type=Float, access=Constructor}
 Sweetener.RandomTimeMax {type=Time}
 Sweetener.RandomTimeMin {type=Time}
 Sweetener.Sounds {type=WaveAsset:Array, access=Constructor}
 UE.Source.OnPlay {type=Trigger}
 Weapon.AmmoRemaining {type=Int32}
 Weapon.ClipSize {type=Int32}
 Weapon.FireRate {type=Float}
 Weapon.FireStartTime {type=Float}
 Weapon.IsAimingDownSights {type=Bool}
]
interfaces [Fort.Character:1.0, Fort.Weapon:1.0, UE.OutputFormat.5dot1:1.0, UE.Source.OneShot:1.0, UE.Source:1.1]
outputs [
 OnGunshotOut {type=Trigger}
 "UE.OutputFormat.5dot1.Audio:0" {type=Audio}
 "UE.OutputFormat.5dot1.Audio:1" {type=Audio}
 "UE.OutputFormat.5dot1.Audio:2" {type=Audio}
 "UE.OutputFormat.5dot1.Audio:3" {type=Audio}
 "UE.OutputFormat.5dot1.Audio:4" {type=Audio}
 "UE.OutputFormat.5dot1.Audio:5" {type=Audio}
 UE.Source.OneShot.OnFinished {type=Trigger}
]
pages [Default {
  input_defaults {
   Character.Affiliation=0
   Character.Health=0
   Character.UniqueID=0
   LowAmmo.FrontLRVolume=1
   LowAmmo.MaxPitch=1.05
   LowAmmo.MaxVolume=1.05
   LowAmmo.MinPitch=0.95
   LowAmmo.MinVolume=0.95
   LowAmmo.Threshold=10
   MainFire.FrontCenterVolume=1
   MainFire.FrontLRVolume=1
   MainFire.LFVolume=1
   MainFire.MaxPitch=1.05
   MainFire.MaxVolume=1.05
   MainFire.MinPitch=0.95
   MainFire.MinVolume=0.95
   MainFire.RearLRVolume=1
   Mech.FrontCenterVolume=1
   Mech.MaxPitch=1.05
   Mech.MaxVolume=1.05
   Mech.MinPitch=0.95
   Mech.MinVolume=0.95
   Mech.RandomTimeMax=0
   Mech.RandomTimeMin=0
   Mech.bEnableMechLayer="<MetaSoundBuilderResult.SUCCEEDED: 0>"
   Mixing.ADSVolumeMultiply=0.8
   Mixing.Master.FrontCenterMasterVolume=1
   Mixing.Master.FrontLRMasterVolume=1
   Mixing.Master.LowFrequencyMasterVolume=1
   Mixing.Master.RearLRMasterVolume=1
   OnStop=null
   OnStopFire.FrontCenterVolume=1
   OnStopFire.FrontLRVolume=1
   OnStopFire.LFVolume=1
   OnStopFire.MaxPitch=1.05
   OnStopFire.MaxVolume=1.05
   OnStopFire.MinPitch=0.95
   OnStopFire.MinVolume=0.95
   OnStopFire.RandomTimeMax=0.6
   OnStopFire.RandomTimeMin=0.6
   OnStopFire.RearLRVolume=1
   Sub.LFVolume=1
   Sub.MaxPitch=1.05
   Sub.MaxVolume=1.05
   Sub.MinPitch=0.95
   Sub.MinVolume=0.95
   Sub.RandomTimeMax=0
   Sub.RandomTimeMin=0
   Sub.bEnable="<MetaSoundBuilderResult.SUCCEEDED: 0>"
   Sweetener.FrontCenterVolume=1
   Sweetener.FrontLRVolume=1
   Sweetener.MaxPitch=1.05
   Sweetener.MaxVolume=1.05
   Sweetener.MinPitch=0.95
   Sweetener.MinVolume=0.7
   Sweetener.RandomTimeMax=0
   Sweetener.RandomTimeMin=0
   UE.Source.OnPlay=null
   Weapon.AmmoRemaining=20
   Weapon.ClipSize=25
   Weapon.FireRate=0.2
   Weapon.FireStartTime=null
   Weapon.IsAimingDownSights=null
   bUseADSSweeteners=null
   LowAmmo.Sounds [/Game/Sounds/Fort_Weapons/LowAmmo/weapon_low_ammo_hi_rifle_05]
   MainFire.Sounds [
    /HopscotchWeaponsGameplay/Items/FlipmagAR/Sounds/FlipMag_Fire_51_Close_01
    /HopscotchWeaponsGameplay/Items/FlipmagAR/Sounds/FlipMag_Fire_51_Close_02
    /HopscotchWeaponsGameplay/Items/FlipmagAR/Sounds/FlipMag_Fire_51_Close_03
    /HopscotchWeaponsGameplay/Items/FlipmagAR/Sounds/FlipMag_Fire_51_Close_04
    /HopscotchWeaponsGameplay/Items/FlipmagAR/Sounds/FlipMag_Fire_51_Close_05
   ]
   Mech.Sounds [
    /HopscotchWeaponsGameplay/Items/FlipmagAR/Sounds/FlipMag_Fire_Close_Mech_01
    /HopscotchWeaponsGameplay/Items/FlipmagAR/Sounds/FlipMag_Fire_Close_Mech_02
    /HopscotchWeaponsGameplay/Items/FlipmagAR/Sounds/FlipMag_Fire_Close_Mech_03
    /HopscotchWeaponsGameplay/Items/FlipmagAR/Sounds/FlipMag_Fire_Close_Mech_04
    /HopscotchWeaponsGameplay/Items/FlipmagAR/Sounds/FlipMag_Fire_Close_Mech_05
   ]
   OnStopFire.Sounds [
    /HopscotchWeaponsGameplay/Items/FlipmagAR/Sounds/FlipMag_Reload_Casing_Impact_01
    /HopscotchWeaponsGameplay/Items/FlipmagAR/Sounds/FlipMag_Reload_Casing_Impact_02
    /HopscotchWeaponsGameplay/Items/FlipmagAR/Sounds/FlipMag_Reload_Casing_Impact_03
    /HopscotchWeaponsGameplay/Items/FlipmagAR/Sounds/FlipMag_Reload_Casing_Impact_04
    /HopscotchWeaponsGameplay/Items/FlipmagAR/Sounds/FlipMag_Reload_Casing_Impact_05
    /HopscotchWeaponsGameplay/Items/FlipmagAR/Sounds/FlipMag_Reload_Casing_Impact_06
    /HopscotchWeaponsGameplay/Items/FlipmagAR/Sounds/FlipMag_Reload_Casing_Impact_07
   ]
   Sub.Sounds [
    /HopscotchWeaponsGameplay/Items/FlipmagAR/Sounds/FlipMag_Fire_Close_Sub_01
    /HopscotchWeaponsGameplay/Items/FlipmagAR/Sounds/FlipMag_Fire_Close_Sub_02
   ]
   Sweetener.Sounds []
  }
  nodes [
   InitVariable.Trigger()
   MSP_Weapon_Fire_Control_0 = /FortMetaSoundWeapon/Items/Metasounds/Legacy/Subgraphs/MSP_Weapon_Fire_Control(OnStop=^OnStop, OnTrigger=^UE.Source.OnPlay, StartingAmmo=^Weapon.AmmoRemaining, FireRate=UE.ConversionFloatToTime(In=^Weapon.FireRate).Out)
   n1 = VariableAccessor.Trigger(Variable=VariableMutator.Trigger(Variable=InitVariable.Trigger().Variable, Value=@MSP_Weapon_Fire_Control_0.OnGunshotOut).Variable)
   n2 = VariableAccessor.Trigger(Variable=@n1.Variable)
   n3 = /FortMetaSoundWeapon/Items/Metasounds/Legacy/Subgraphs/MSS_Weapon_Fire_Main_5_1(
    MaxPitch=^MainFire.MaxPitch
    MaxVolume=^MainFire.MaxVolume
    MinPitch=^MainFire.MinPitch
    MinVolume=^MainFire.MinVolume
    FireSounds=^MainFire.Sounds
    UE.Source.OnPlay=@n2.Value
   )
   n4 = TriggerCompare.Clamp:Bool(A=^Weapon.IsAimingDownSights, Compare=@n2.Value) {also {defaults {B="<MetaSoundBuilderResult.SUCCEEDED: 0>"}}}
   n5 = Value.Value:Float(Target Value=^Mixing.ADSVolumeMultiply, Reset=@n4.False, Set=@n4.True) {also {defaults {Init Value=1}}}
   n6 = VariableAccessor.Trigger(Variable=@n2.Variable)
   n7 = VariableAccessor.Trigger(Variable=@n6.Variable)
   n8 = VariableAccessor.Trigger(Variable=@n7.Variable)
   n9 = VariableAccessor.Trigger(Variable=@n8.Variable)
   n10 = MapRange.MapRange:Float(In=Convert.Float:Bool(Bool=^Weapon.IsAimingDownSights).Float) {also {defaults {Out Range A=1, Out Range B=1.3}}}
   AudioMixerAudioMixerMono8_0 = "AudioMixer.Audio Mixer (Mono, 8)"(
    Gain 2=^MainFire.LFVolume
    Gain 4=^OnStopFire.LFVolume
    Gain 0=^Sub.LFVolume
    In 2=UE.Multiply:Audio by Float(PrimaryOperand=@n3."UE.OutputFormat.5dot1.Audio:3", AdditionalOperands=@n5.Output Value).Out
    In 0=UE.Multiply:Audio by Float(
     PrimaryOperand=/FortMetaSoundWeapon/Items/Metasounds/Legacy/Subgraphs/MSP_Weapon_Fire_Mech(
      MaxPitch=^Sub.MaxPitch
      MaxVolume=^Sub.MaxVolume
      MinPitch=^Sub.MinPitch
      MinVolume=^Sub.MinVolume
      RandomTimeMax=^Sub.RandomTimeMax
      RandomTimeMin=^Sub.RandomTimeMin
      MechLayerSounds=^Sub.Sounds
      bEnable=^Sub.bEnable
      OnPlay=@n9.Value
     ).Out Mono
     AdditionalOperands=@n10.Out Value
    ).Out
   )
   Multiply_0 = UE.Multiply:Audio by Float(AdditionalOperands=^Mixing.Master.LowFrequencyMasterVolume, PrimaryOperand=@AudioMixerAudioMixerMono8_0.Out)
   AudioMixerAudioMixerStereo8_0 = "AudioMixer.Audio Mixer (Stereo, 8)"(
    Gain 0=^MainFire.RearLRVolume
    Gain 2=^OnStopFire.RearLRVolume
    In 0 R=UE.Multiply:Audio by Float(PrimaryOperand=@n3."UE.OutputFormat.5dot1.Audio:5", AdditionalOperands=@n5.Output Value).Out
    In 0 L=UE.Multiply:Audio by Float(PrimaryOperand=@n3."UE.OutputFormat.5dot1.Audio:4", AdditionalOperands=@n5.Output Value).Out
   )
   Multiply_1 = UE.Multiply:Audio by Float(AdditionalOperands=^Mixing.Master.RearLRMasterVolume, PrimaryOperand=@AudioMixerAudioMixerStereo8_0.Out R)
   n11 = /FortMetaSoundWeapon/Items/Metasounds/Legacy/Subgraphs/MSP_Weapon_Fire_Sweetener(
    MaxPitch=^Sweetener.MaxPitch
    MaxVolume=^Sweetener.MaxVolume
    MinPitch=^Sweetener.MinPitch
    MinVolume=^Sweetener.MinVolume
    RandomTimeMax=^Sweetener.RandomTimeMax
    RandomTimeMin=^Sweetener.RandomTimeMin
    SweetenerLayerSounds=^Sweetener.Sounds
    bEnable=^Weapon.IsAimingDownSights
    OnPlay=TriggerCompare.Clamp:Bool(A=^bUseADSSweeteners, Compare=@n7.Value) {also {defaults {B="<MetaSoundBuilderResult.SUCCEEDED: 0>"}}}.True
   )
   n12 = InitVariable.Int32()
   n13 = /FortMetaSoundWeapon/Items/Metasounds/Legacy/Subgraphs/MSS_Weapon_Fire_Low(
    MaxPitch=^LowAmmo.MaxPitch
    MaxVolume=^LowAmmo.MaxVolume
    MinPitch=^LowAmmo.MinPitch
    MinVolume=^LowAmmo.MinVolume
    FireLowAmmoSounds=^LowAmmo.Sounds
    LowAmmoThreshold=^LowAmmo.Threshold
    CurrentAmmo=VariableAccessor.Int32(Variable=VariableMutator.Int32(Variable=@n12.Variable, Value=@MSP_Weapon_Fire_Control_0.CurrentAmmoOut).Variable).Value
    UE.Source.OnPlay=@n1.Value
   )
   n14 = &MSS_PlayRandomSound_Mono(
    MaxPitch=^OnStopFire.MaxPitch
    MaxVolume=^OnStopFire.MaxVolume
    MinPitch=^OnStopFire.MinPitch
    MinVolume=^OnStopFire.MinVolume
    Sounds=^OnStopFire.Sounds
    UE.Source.OnPlay=UE.RandomTime(Max=^OnStopFire.RandomTimeMax, Min=^OnStopFire.RandomTimeMin, Next=@n6.Value).On Next
   )
   n15 = UE.Multiply:Audio by Float(PrimaryOperand=@n14."UE.OutputFormat.Mono.Audio:0", AdditionalOperands=@n10.Out Value)
   AudioMixerAudioMixerStereo8_1 = "AudioMixer.Audio Mixer (Stereo, 8)"(
    Gain 1=^LowAmmo.FrontLRVolume
    Gain 0=^MainFire.FrontLRVolume
    Gain 2=^OnStopFire.FrontLRVolume
    Gain 3=^Sweetener.FrontLRVolume
    In 3 L=@n11.Out Left
    In 3 R=@n11.Out Right
    In 1 L=@n13."UE.OutputFormat.Stereo.Audio:0"
    In 1 R=@n13."UE.OutputFormat.Stereo.Audio:1"
    In 0 R=UE.Multiply:Audio by Float(PrimaryOperand=@n3."UE.OutputFormat.5dot1.Audio:1", AdditionalOperands=@n5.Output Value).Out
    In 2 L=@n15.Out
    In 2 R=@n15.Out
    In 0 L=UE.Multiply:Audio by Float(PrimaryOperand=@n3."UE.OutputFormat.5dot1.Audio:0", AdditionalOperands=@n5.Output Value).Out
   )
   Multiply_10 = UE.Multiply:Audio by Float(AdditionalOperands=^Mixing.Master.FrontLRMasterVolume, PrimaryOperand=@AudioMixerAudioMixerStereo8_1.Out R)
   Multiply_14 = UE.Multiply:Audio by Float(AdditionalOperands=^Mixing.Master.FrontLRMasterVolume, PrimaryOperand=@AudioMixerAudioMixerStereo8_1.Out L)
   Multiply_6 = UE.Multiply:Audio by Float(AdditionalOperands=^Mixing.Master.RearLRMasterVolume, PrimaryOperand=@AudioMixerAudioMixerStereo8_0.Out L)
   AudioMixerAudioMixerMono8_1 = "AudioMixer.Audio Mixer (Mono, 8)"(
    Gain 0=^MainFire.FrontCenterVolume
    Gain 1=^Mech.FrontCenterVolume
    Gain 2=^OnStopFire.FrontCenterVolume
    Gain 3=^Sweetener.FrontCenterVolume
    In 0=UE.Multiply:Audio by Float(PrimaryOperand=@n3."UE.OutputFormat.5dot1.Audio:2", AdditionalOperands=@n5.Output Value).Out
    In 1=UE.Multiply:Audio by Float(
     PrimaryOperand=&MSP_Weapon_Fire_Mech(
      MaxPitch=^Mech.MaxPitch
      MaxVolume=^Mech.MaxVolume
      MinPitch=^Mech.MinPitch
      MinVolume=^Mech.MinVolume
      RandomTimeMax=^Mech.RandomTimeMax
      RandomTimeMin=^Mech.RandomTimeMin
      MechLayerSounds=^Mech.Sounds
      bEnable=^Mech.bEnableMechLayer
      OnPlay=@n8.Value
     ).Out Mono
     AdditionalOperands=@n10.Out Value
    ).Out
   )
   Multiply_9 = UE.Multiply:Audio by Float(AdditionalOperands=^Mixing.Master.FrontCenterMasterVolume, PrimaryOperand=@AudioMixerAudioMixerMono8_1.Out)
   TriggerAccumulatorTriggerAccumulate3_0 = "TriggerAccumulator.Trigger Accumulate (3)"(In 2=@n14.UE.Source.OneShot.OnFinished, In 1=@n13.UE.Source.OneShot.OnFinished, In 0=@n3.UE.Source.OneShot.OnFinished)
   VariableAccessor.Trigger(Variable=VariableMutator.Trigger(Variable=InitVariable.Trigger().Variable, Value=@MSP_Weapon_Fire_Control_0.OnOutOfAmmo).Variable)
   VariableMutator.Bool(Variable=InitVariable.Bool().Variable, Value=@MSP_Weapon_Fire_Control_0.bIsFiring)
  ]
  links [
   {dst="@OnGunshotOut", src=MSP_Weapon_Fire_Control_0.OnGunshotOut}
   {dst="@UE.OutputFormat.5dot1.Audio:3", src=Multiply_0.Out}
   {dst="@UE.OutputFormat.5dot1.Audio:5", src=Multiply_1.Out}
   {dst="@UE.OutputFormat.5dot1.Audio:1", src=Multiply_10.Out}
   {dst="@UE.OutputFormat.5dot1.Audio:0", src=Multiply_14.Out}
   {dst="@UE.OutputFormat.5dot1.Audio:4", src=Multiply_6.Out}
   {dst="@UE.OutputFormat.5dot1.Audio:2", src=Multiply_9.Out}
   {dst="@UE.Source.OneShot.OnFinished", src=TriggerAccumulatorTriggerAccumulate3_0.Out}
  ]
 }]
properties {
 assetImportData=&self:AssetImportData
 numChannels=6
 outputFormat=FiveDotOne
 qualitySetting=Fort_Weapons
 sampleRate=48000
 soundClassObject=/Game/Sounds/FortSoundClassesAndModes/Fort_Events_Weapons_1P
 concurrencyOverrides {maxCount=2, resolutionRule=StopOldest}
 concurrencySet [&Weapon_Gunfire_Local_Conc, &Weapon_Gunfire_Global_Conc]
}
