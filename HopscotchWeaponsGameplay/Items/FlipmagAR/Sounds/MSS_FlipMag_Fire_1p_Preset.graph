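# MetaSound Source preset for the first-person FlipMag AR fire sound, inheriting its graph from MSS_FlipMag_Weapon_Fire_1p and supplying waves and mix values.
# It fills 5.1 main fire, mech, sub and low-ammo layers, plus casing impacts for stop-fire; the sweetener array is empty.
# Inputs cover randomized pitch, volume and timing, per-channel levels, and Fort weapon state such as ammo and aiming.
# Owner-limited concurrency allows two instances, stopping the oldest.
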MetaSoundSource /HopscotchWeaponsGameplay/Items/FlipmagAR/Sounds/MSS_FlipMag_Fire_1p_Preset
preset_parent=/HopscotchWeaponsGameplay/Items/FlipmagAR/Sounds/Settings/MSS_FlipMag_Weapon_Fire_1p
inputs [
 bUseADSSweeteners {type=Bool, access=Constructor}
 Character.Affiliation {type=Enum:FortAudioAffiliation}
 Character.Health {type=Float}
 Character.UniqueID {type=Int32}
 LowAmmo.FrontLRVolume {type=Float}
 LowAmmo.MaxPitch {type=Float, access=Constructor}
 LowAmmo.MaxVolume {type=Float, access=Constructor}
 LowAmmo.MinPitch {type=Float, access=Constructor}
 LowAmmo.MinVolume {type=Float, access=Constructor}
 LowAmmo.Sounds {type=WaveAsset:Array, access=Constructor}
 LowAmmo.Threshold {type=Int32}
 MainFire.FrontCenterVolume {type=Float}
 MainFire.FrontLRVolume {type=Float}
 MainFire.LFVolume {type=Float}
 MainFire.MaxPitch {type=Float, access=Constructor}
 MainFire.MaxVolume {type=Float, access=Constructor}
 MainFire.MinPitch {type=Float, access=Constructor}
 MainFire.MinVolume {type=Float, access=Constructor}
 MainFire.RearLRVolume {type=Float, access=Constructor}
 MainFire.Sounds {type=WaveAsset:Array, access=Constructor}
 Mech.bEnableMechLayer {type=Bool}
 Mech.FrontCenterVolume {type=Float}
 Mech.MaxPitch {type=Float, access=Constructor}
 Mech.MaxVolume {type=Float, access=Constructor}
 Mech.MinPitch {type=Float, access=Constructor}
 Mech.MinVolume {type=Float, access=Constructor}
 Mech.RandomTimeMax {type=Time, access=Constructor}
 Mech.RandomTimeMin {type=Time, access=Constructor}
 Mech.Sounds {type=WaveAsset:Array, access=Constructor}
 Mixing.ADSVolumeMultiply {type=Float}
 Mixing.Master.FrontCenterMasterVolume {type=Float}
 Mixing.Master.FrontLRMasterVolume {type=Float}
 Mixing.Master.LowFrequencyMasterVolume {type=Float}
 Mixing.Master.RearLRMasterVolume {type=Float}
 OnStop {type=Trigger}
 OnStopFire.FrontCenterVolume {type=Float}
 OnStopFire.FrontLRVolume {type=Float}
 OnStopFire.LFVolume {type=Float}
 OnStopFire.MaxPitch {type=Float, access=Constructor}
 OnStopFire.MaxVolume {type=Float, access=Constructor}
 OnStopFire.MinPitch {type=Float, access=Constructor}
 OnStopFire.MinVolume {type=Float, access=Constructor}
 OnStopFire.RandomTimeMax {type=Time}
 OnStopFire.RandomTimeMin {type=Time}
 OnStopFire.RearLRVolume {type=Float}
 OnStopFire.Sounds {type=WaveAsset:Array, access=Constructor}
 Sub.bEnable {type=Bool}
 Sub.LFVolume {type=Float}
 Sub.MaxPitch {type=Float, access=Constructor}
 Sub.MaxVolume {type=Float, access=Constructor}
 Sub.MinPitch {type=Float, access=Constructor}
 Sub.MinVolume {type=Float, access=Constructor}
 Sub.RandomTimeMax {type=Time}
 Sub.RandomTimeMin {type=Time}
 Sub.Sounds {type=WaveAsset:Array, access=Constructor}
 Sweetener.FrontCenterVolume {type=Float}
 Sweetener.FrontLRVolume {type=Float}
 Sweetener.MaxPitch {type=Float, access=Constructor}
 Sweetener.MaxVolume {type=Float, access=Constructor}
 Sweetener.MinPitch {type=Float, access=Constructor}
 Sweetener.MinVolume {type=Float, access=Constructor}
 Sweetener.RandomTimeMax {type=Time}
 Sweetener.RandomTimeMin {type=Time}
 Sweetener.Sounds {type=WaveAsset:Array, access=Constructor}
 UE.Source.OnPlay {type=Trigger}
 Weapon.AmmoRemaining {type=Int32}
 Weapon.ClipSize {type=Int32}
 Weapon.FireRate {type=Float}
 Weapon.FireStartTime {type=Float}
 Weapon.IsAimingDownSights {type=Bool}
]
interfaces [Fort.Character:1.0, Fort.Weapon:1.0, UE.OutputFormat.5dot1:1.0, UE.Source.OneShot:1.0, UE.Source:1.1]
outputs [
 OnGunshotOut {type=Trigger}
 "UE.OutputFormat.5dot1.Audio:0" {type=Audio}
 "UE.OutputFormat.5dot1.Audio:1" {type=Audio}
 "UE.OutputFormat.5dot1.Audio:2" {type=Audio}
 "UE.OutputFormat.5dot1.Audio:3" {type=Audio}
 "UE.OutputFormat.5dot1.Audio:4" {type=Audio}
 "UE.OutputFormat.5dot1.Audio:5" {type=Audio}
 UE.Source.OneShot.OnFinished {type=Trigger}
]
pages [Default {input_defaults {
   Character.Affiliation=0
   Character.Health=0
   Character.UniqueID=0
   LowAmmo.FrontLRVolume=1
   LowAmmo.MaxPitch=1.05
   LowAmmo.MaxVolume=1.05
   LowAmmo.MinPitch=0.95
   LowAmmo.MinVolume=0.95
   LowAmmo.Threshold=10
   MainFire.FrontCenterVolume=1
   MainFire.FrontLRVolume=1
   MainFire.LFVolume=1
   MainFire.MaxPitch=1.05
   MainFire.MaxVolume=1.05
   MainFire.MinPitch=0.95
   MainFire.MinVolume=0.95
   MainFire.RearLRVolume=1
   Mech.FrontCenterVolume=0.1856
   Mech.MaxPitch=1.05
   Mech.MaxVolume=1.05
   Mech.MinPitch=0.95
   Mech.MinVolume=0.95
   Mech.RandomTimeMax=0.1
   Mech.RandomTimeMin=0
   Mech.bEnableMechLayer="<MetaSoundBuilderResult.SUCCEEDED: 0>"
   Mixing.ADSVolumeMultiply=0.8
   Mixing.Master.FrontCenterMasterVolume=0.85
   Mixing.Master.FrontLRMasterVolume=1
   Mixing.Master.LowFrequencyMasterVolume=1.15
   Mixing.Master.RearLRMasterVolume=0.9
   OnStop=null
   OnStopFire.FrontCenterVolume=1
   OnStopFire.FrontLRVolume=0.5
   OnStopFire.LFVolume=1
   OnStopFire.MaxPitch=1.05
   OnStopFire.MaxVolume=1.05
   OnStopFire.MinPitch=0.95
   OnStopFire.MinVolume=0.95
   OnStopFire.RandomTimeMax=0.6
   OnStopFire.RandomTimeMin=0.6
   OnStopFire.RearLRVolume=1
   Sub.LFVolume=1
   Sub.MaxPitch=1.05
   Sub.MaxVolume=1.05
   Sub.MinPitch=0.95
   Sub.MinVolume=0.95
   Sub.RandomTimeMax=0
   Sub.RandomTimeMin=0
   Sub.bEnable="<MetaSoundBuilderResult.SUCCEEDED: 0>"
   Sweetener.FrontCenterVolume=1
   Sweetener.FrontLRVolume=1
   Sweetener.MaxPitch=1.05
   Sweetener.MaxVolume=1.05
   Sweetener.MinPitch=0.95
   Sweetener.MinVolume=0.7
   Sweetener.RandomTimeMax=0
   Sweetener.RandomTimeMin=0
   UE.Source.OnPlay=null
   Weapon.AmmoRemaining=20
   Weapon.ClipSize=25
   Weapon.FireRate=0.2
   Weapon.FireStartTime=null
   Weapon.IsAimingDownSights=null
   bUseADSSweeteners=null
   LowAmmo.Sounds [/Game/Sounds/Fort_Weapons/LowAmmo/weapon_low_ammo_hi_rifle_05]
   MainFire.Sounds [
    &/FlipMag_Fire_51_Close_01
    &/FlipMag_Fire_51_Close_02
    &/FlipMag_Fire_51_Close_03
    &/FlipMag_Fire_51_Close_04
    &/FlipMag_Fire_51_Close_05
   ]
   Mech.Sounds [
    &/FlipMag_Fire_Close_Mech_01
    &/FlipMag_Fire_Close_Mech_02
    &/FlipMag_Fire_Close_Mech_03
    &/FlipMag_Fire_Close_Mech_04
    &/FlipMag_Fire_Close_Mech_05
   ]
   OnStopFire.Sounds [
    &/FlipMag_Reload_Casing_Impact_01
    &/FlipMag_Reload_Casing_Impact_02
    &/FlipMag_Reload_Casing_Impact_03
    &/FlipMag_Reload_Casing_Impact_04
    &/FlipMag_Reload_Casing_Impact_05
    &/FlipMag_Reload_Casing_Impact_06
    &/FlipMag_Reload_Casing_Impact_07
   ]
   Sub.Sounds [
    &/FlipMag_Fire_Close_Sub_01
    &/FlipMag_Fire_Close_Sub_02
   ]
   Sweetener.Sounds []
  }}]
properties {
 assetImportData=&self:AssetImportData
 bOverrideConcurrency=true
 numChannels=6
 outputFormat=FiveDotOne
 qualitySetting=Fort_Weapons
 sampleRate=48000
 soundClassObject=&Fort_WeaponRanged_Fire_1P_PSM
 volume=0.9
 concurrencyOverrides {bLimitToOwner=true, maxCount=2, resolutionRule=StopOldest}
 concurrencySet [&Weapon_Gunfire_Local_Conc, &Weapon_Gunfire_Global_Conc]
}
